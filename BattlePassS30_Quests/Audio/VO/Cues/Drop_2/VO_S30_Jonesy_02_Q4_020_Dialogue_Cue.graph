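# Sound cue for a Season 30 Battle Pass quest voice line by Jonesy (drop 2, Q4, line 020).
# A single dialogue player node plays the dialogue wave through the shared audio-log voice as speaker, with no targets.
# It routes to the narrator VO sound class and uses shared audio-log attenuation settings with a very large max distance.
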SoundCue /BattlePassS30_Quests/Audio/VO/Cues/Drop_2/VO_S30_Jonesy_02_Q4_020_Dialogue_Cue
root=DialoguePlayer_0
nodes [{
  id=DialoguePlayer_0
  properties {dialogueWaveParameter {
    dialogueWave=/BattlePassS30_Quests/Audio/VO/Waves/Jonesy/VO_S30_Jonesy_02_Q4_020_Dialogue
    context {speaker=/BattlePassBase/Audio/VO/AudioLogVO_DV, targets []}
   }}
 }]
properties {
 attenuationSettings=/BattlePassBase/Audio/SFX/Fort_AudioLogs_Att
 maxDistance=2097152
 soundClassObject=/Game/Sounds/FortSoundClassesAndModes/Fort_VO_Narrator
 volumeMultiplier=1
}
